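; Placeable local fog actor for the Clyde time-of-day light sources, drawing a fog material on a sphere component.
; The construction script creates a dynamic instance of the sphere's slot 0 material, stores it as Fog_Mat and assigns it back.
; It then pushes the actor's density, color, bump, texture power and tiling, high/low fade and depth fade variables into material parameters.
; BeginPlay, overlap and Tick are empty, so nothing updates at runtime.
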(event EventBeginPlay)

(event Collision|EventActorBeginOverlap
  (params (OtherActor :type Actor%20Object%20Reference)))

(event EventTick
  (params (DeltaSeconds :type Float%20%28single-precision%29)))

(fn ConstructionScript
  (bind _fog_mat (Variables|Default|GetFog_Mat))
  ;[1 1 1 0.5] Set_Emissive_MAT
  (bind _returnvalue (Rendering|Material|CreateDynamicMaterialInstance (Rendering|Material|GetMaterial (Variables|Default|GetSphere) 0) "None" "None" :!class /Script/Engine.KismetMaterialLibrary))
  (bind _output_get (Variables|Default|SetFog_Mat _returnvalue))
  (Rendering|Material|SetMaterial (Variables|Default|GetSphere) 0 _output_get)
  (Rendering|Material|SetScalarParameterValue _fog_mat "density" (Variables|Default|GetDensity) :!class /Script/Engine.MaterialInstanceDynamic)
  (Rendering|Material|SetVectorParameterValue _fog_mat "Fog_Color" (Variables|Default|GetFog_Color) :!class /Script/Engine.MaterialInstanceDynamic)
  (Rendering|Material|SetScalarParameterValue _fog_mat "bump" (Variables|Default|GetBump) :!class /Script/Engine.MaterialInstanceDynamic)
  (Rendering|Material|SetScalarParameterValue _fog_mat "Pow" (Variables|Default|GetTex_Power) :!class /Script/Engine.MaterialInstanceDynamic)
  (Rendering|Material|SetScalarParameterValue _fog_mat "tile" (Variables|Default|GetTex_Tile) :!class /Script/Engine.MaterialInstanceDynamic)
  (Rendering|Material|SetScalarParameterValue _fog_mat "Fade_High" (Variables|Default|GetFade_High) :!class /Script/Engine.MaterialInstanceDynamic)
  (Rendering|Material|SetScalarParameterValue _fog_mat "Fade_Low" (Variables|Default|GetFade_Low) :!class /Script/Engine.MaterialInstanceDynamic)
  (Rendering|Material|SetScalarParameterValue _fog_mat "Fade" (Variables|Default|GetDepth_Fade) :!class /Script/Engine.MaterialInstanceDynamic))
