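# Light function material for the ClayPlug meeting light among Clyde's time-of-day light sources, shaping a soft ring pattern.
# Two sphere masks centered in UV space, with parameterized outer and inner radii, form the ring by multiplying the outer mask with the inverted inner one.
# A Ring_Brightness parameter then scales the result.
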Material /Clyde/TimeOfDay/Clyde_Light_Sources/ClayPlug_MeetingLight_LF
graph {
 properties {bAutomaticallySetUsageInEditor=false, materialDomain=MD_LightFunction, usedShadingModels=Default Lit}
 nodes [
  n1 = TextureCoordinate()
  n2 = Constant2Vector() {g=0.5, r=0.5}
  n3 = ScalarParameter() {defaultValue=0.5, parameterName=Radius_Outer}
  n4 = SphereMask(A=@n1, B=@n2, Radius=ScalarParameter() {defaultValue=0.7, parameterName=Radius_Inner}) {hardnessPercent=10}
  Multiply(
   A=Multiply(A=SphereMask(A=@n1, B=@n2, Radius=@n3) {attenuationRadius=0.5, hardnessPercent=0}, B=OneMinus(None=@n4))
   B=ScalarParameter() {defaultValue=5, parameterName=Ring_Brightness}
  ) {constB=5}
 ]
}
