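# Light function material for blinking power plant lights in the Clyde time-of-day light sources.
# It computes a sine wave from Time divided by the Time_Division parameter and offset by Time_Add.
# Time_Division sets the blink rate and Time_Add shifts the phase, so material instances can stagger their blinking.
# The sine is neither clamped nor remapped, so the wave swings below zero for part of each cycle.
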Material /Clyde/TimeOfDay/Clyde_Light_Sources/LF_PowerPlant_Light_Blink
graph {
 properties {bAutomaticallySetUsageInEditor=false, materialDomain=MD_LightFunction, usedShadingModels=Default Lit}
 nodes [
  n1 = Time()
  n2 = ScalarParameter() {defaultValue=3, parameterName=Time_Division}
  Sine(None=Add(A=Divide(A=@n1, B=@n2), B=ScalarParameter() {parameterName=Time_Add}))
 ]
}
